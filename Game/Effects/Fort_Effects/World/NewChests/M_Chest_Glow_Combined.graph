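# Translucent, unlit, two-sided glow material for chest effects, with tinted emissive scaled by a custom primitive data brightness.
# World position offset yaws the mesh around its orientation axis toward the camera, then applies size and local offset, with a Juno switch choosing the scaling path.
# Opacity combines a UV sphere mask, gradients, pivot-height fade and Anim scalar.
# Deferred shading adds depth and near-camera fades, while other paths use a mobile opacity multiplier.
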Material /Game/Effects/Fort_Effects/World/NewChests/M_Chest_Glow_Combined
graph {
 properties {blendMode=BLEND_Translucent, pixelDepthOffsetMode=PDOM_Legacy, shadingModel=MSM_Unlit, twoSided=true, usedShadingModels=Unlit}
 nodes [
  n1 = ObjectOrientation()
  Abs(None=@n1)
  MaterialFunctionCall(Emissive=Multiply(
    A=VectorParameter() {parameterName=Glow_Colour, defaultValue=<rgba 7 2.895 0 4>}.RGB
    B=ScalarParameter() {bUseCustomPrimitiveData=true, defaultValue=1, parameterName=Scaled_Brightness Primitive, primitiveDataIndex=1}
   )) {materialFunction=&MF_FX_EmissiveAdjustments}
  n2 = Normalize(VectorInput=Subtract(A=ObjectPositionWS(), B=CameraPositionWS()))
  n3 = Arctangent2Fast(Y=ComponentMask(None=@n2) {g=true}, X=ComponentMask(None=@n2) {r=true})
  n4 = MaterialFunctionCall(Multiplier=Constant() {r=2}) {materialFunction=&Pi}
  n5 = RotateAboutAxis(
   NormalizedRotationAxis=@n1
   RotationAngle=Add(A=Divide(A=@n3, B=@n4.Result), B=Constant() {r=0.25}) {constA=0.25, constB=0.25}
   PivotPoint=MaterialFunctionCall() {materialFunction=&ObjectPivotPoint}.Object Pivot Location
   Position=WorldPosition().XYZ
  )
  n6 = MaterialFunctionCall() {materialFunction=&ObjectPivotPoint}
  n7 = Add(A=@n5, B=WorldPosition() {worldPositionShaderOffset=WPT_ExcludeAllShaderOffsets}.XYZ)
  n8 = TransformPosition(None=@n7) {transformSourceType=TRANSFORMPOSSOURCE_World}
  n9 = StaticSwitchParameter(True=Multiply(A=@n8, B=VectorParameter() {parameterName=Scaled_MeshSizeXYZ, defaultValue=<rgba 1 1 1 1>}.RGB) {constB=1.02}, False=@n8) {parameterName=Juno}
  n10 = Add(A=@n5, B=Subtract(A=Add(A=@n6.Object Pivot Location, B=@n9), B=@n7))
  n11 = ObjectPositionWS()
  n12 = WorldPosition()
  n13 = VectorParameter() {parameterName=Scaled_MeshSizeXYZ, defaultValue=<rgba 1 1 1 1>}
  n14 = Add(A=@n10, B=Multiply(A=Subtract(A=@n11, B=@n12.XYZ), B=OneMinus(None=@n13.RGB)))
  n15 = VectorParameter() {parameterName=ObjectLocationOffset, defaultValue=<rgba 0 0 0 1>}
  NamedRerouteDeclaration(None=Add(A=StaticSwitchParameter(True=@n10, False=@n14) {parameterName=Juno}, B=Transform(None=@n15.RGB) {transformSourceType=TRANSFORMSOURCE_Local})) {name=WPO, nodeColor=<rgba 0.3125 1 0 1>}
  n16 = TextureCoordinate() {coordinateIndex=1}
  n17 = ComponentMask(None=VectorParameter() {parameterName=Glow Sphere Mask B, defaultValue=<rgba 0.5 0.5 0 1>}.RGB) {g=true, r=true}
  n18 = ScalarParameter() {defaultValue=0.5, parameterName=Radius}
  n19 = ScalarParameter() {parameterName=Hardness}
  n20 = Multiply(
   A=ScalarParameter() {defaultValue=4, parameterName=Glow_Opacity}
   B=Power(Base=SphereMask(A=@n16, B=@n17, Radius=@n18, Hardness=@n19), Exp=ScalarParameter() {defaultValue=3, parameterName=Exp})
  )
  n21 = ScalarParameter() {defaultValue=0.0175, parameterName=ForwardMobileOpacityMultiplier}
  n22 = Multiply(A=@n20, B=DepthFade(FadeDistance=ScalarParameter() {defaultValue=5, parameterName=Depth Fade}))
  n23 = MaterialFunctionCall(Distance=ScalarParameter() {parameterName=Camera Fade}) {materialFunction=&Near_Camera_Fade}
  n24 = OneMinus(None=ComponentMask(None=TextureCoordinate()) {g=true})
  n25 = ScalarParameter() {defaultValue=3, parameterName=Exp}
  n26 = OneMinus(None=ComponentMask(None=TextureCoordinate() {coordinateIndex=1}) {g=true})
  n27 = ScalarParameter() {defaultValue=3, parameterName=ExpCoord1}
  n28 = Multiply(
   A=ShadingPathSwitch(Default=Multiply(A=@n20, B=@n21) {constB=0.0175}, Deferred=Multiply(A=@n22, B=@n23.Result))
   B=Multiply(A=Power(Base=@n24, Exp=@n25), B=Power(Base=@n26, Exp=@n27))
  )
  n29 = ComponentMask(None=Subtract(
    A=WorldPosition().XYZ
    B=MaterialFunctionCall() {materialFunction=&ObjectPivotPoint}.Object Pivot Location
   )) {b=true}
  n30 = Saturate(None=Multiply(A=Abs(None=@n29), B=ScalarParameter() {defaultValue=0.05, parameterName=FadeBottomWorldScaleMultiplier}) {constB=0.01})
  NamedRerouteDeclaration(None=Multiply(A=Multiply(A=@n28, B=@n30), B=ScalarParameter() {defaultValue=1, parameterName=Anim})) {name=Glow Opacity, nodeColor=<rgba 0 1 0.25 1>}
  NamedRerouteUsage()
  NamedRerouteUsage()
  ScalarParameter() {defaultValue=1, parameterName=Scaled_MeshSize}
 ]
}
