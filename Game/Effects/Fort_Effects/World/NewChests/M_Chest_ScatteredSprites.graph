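# Additive, unlit sprite material for new-chest effects, scattering glowing flare sprites around the object.
# Color combines a desaturated sphere-mask core with a spiky flare texture; alpha applies bottom fade and deferred-only depth fade.
# World position offset billboards quads and adds vertical bobbing and looping radial drift, using UV1 for per-sprite variation.
# The Juno switch selects alternate flicker, motion and alpha paths; custom primitive data drives opacity and collapses the mesh to hide it.
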Material /Game/Effects/Fort_Effects/World/NewChests/M_Chest_ScatteredSprites
graph {
 properties {blendMode=BLEND_Additive, pixelDepthOffsetMode=PDOM_Legacy, shadingModel=MSM_Unlit, twoSided=true, usedShadingModels=Unlit}
 nodes [
  Constant() {r=1}
  n1 = Constant2Vector() {g=-1, r=0.05}
  n2 = ScalarParameter() {defaultValue=1, parameterName=FlikeringSpeed}
  n3 = TextureSample(UVs=Panner(Coordinate=TextureCoordinate() {coordinateIndex=1}, Speed=Multiply(A=@n1, B=@n2)) {speedX=0.05, speedY=-1}) {texture=/Game/Effects/Fort_Effects/Textures/Tile/T_Inky_Smoke_Tile2}
  n4 = ScalarParameter() {defaultValue=5}
  LinearInterpolate(
   B=LinearInterpolate(
    A=ScalarParameter() {defaultValue=20, parameterName=Scale}
    Alpha=ScalarParameter() {bUseCustomPrimitiveData=true, parameterName=WPO_HideMesh Primitive}
   ) {constA=20, constB=0}
   Alpha=Saturate(None=Multiply(A=@n3.R, B=@n4))
  )
  n5 = ScalarParameter() {defaultValue=25, parameterName=CentertPower}
  n6 = Constant() {r=0.5}
  n7 = SphereMask(A=TextureCoordinate(), B=@n6, Radius=@n6) {hardnessPercent=0}
  n8 = LinearInterpolate(
   A=VectorParameter() {parameterName=SpriteColor_Low, defaultValue=<rgba 4 1.625 0 1>}.RGB
   B=VectorParameter() {parameterName=SpriteColor_High, defaultValue=<rgba 4 3 2.25 1>}.RGB
   Alpha=Clamp(None=NamedRerouteUsage())
  )
  n9 = Multiply(
   A=Multiply(A=Power(Base=@n7, Exp=@n5) {constExponent=25}, B=ScalarParameter() {defaultValue=2, parameterName=CenterBrightness}) {constB=2}
   B=Desaturation(None=@n8, Fraction=ScalarParameter() {defaultValue=0.3, parameterName=CenterDesaturation})
  )
  n10 = Power(Base=@n7, Exp=ScalarParameter() {defaultValue=4, parameterName=EdgePower}) {constExponent=4}
  n11 = ScalarParameter() {defaultValue=0.15, parameterName=EdgeBrightness}
  n12 = Add(A=TextureCoordinate(), B=ComponentMask(None=VectorParameter() {defaultValue=<rgba 0 0 0 1>}.RGB) {g=true, r=true})
  n13 = Multiply(
   A=@n8
   B=Add(
    A=Multiply(A=@n10, B=@n11) {constB=0.15}
    B=TextureSampleParameter2D(UVs=@n12) {parameterName=FlareMask, texture=/Game/Effects/Fort_Effects/Textures/Flares/T_Flare_Spikeball}.RGB
   )
  )
  Multiply(
   A=MaterialFunctionCall(Emissive=Add(A=@n9, B=@n13)) {materialFunction=&MF_FX_EmissiveAdjustments}.Result
   B=ScalarParameter() {defaultValue=1, parameterName=Anim}
  )
  n14 = Multiply(
   A=ScalarParameter() {defaultValue=1, parameterName=Material Opacity}
   B=LinearInterpolate(
    B=ScalarParameter() {bUseCustomPrimitiveData=true, defaultValue=1, parameterName=Material Opacity Primitive, primitiveDataIndex=2}
    Alpha=ScalarParameter() {defaultValue=1, parameterName=UseMaterialOpacityPrimitive}
   ) {constA=1}
  )
  n15 = ScalarParameter() {parameterName=DepthFade}
  n16 = TextureCoordinate() {coordinateIndex=1}
  n17 = Power(Base=OneMinus(None=ComponentMask(None=@n16) {g=true}), Exp=ScalarParameter() {defaultValue=5, parameterName=Bottom_Fade_Contrast})
  n18 = ScalarParameter() {defaultValue=5, parameterName=Bottom Fade Scale}
  n19 = Multiply(A=ShadingPathSwitch(Default=@n14, Deferred=DepthFade(Opacity=@n14, FadeDistance=@n15)), B=Saturate(None=Multiply(A=@n17, B=@n18)))
  n20 = Panner(Coordinate=TextureCoordinate() {coordinateIndex=1, uTiling=0.25, vTiling=0.25}) {speedY=-0.5}
  n21 = Saturate(None=Multiply(A=TextureSample(UVs=@n20) {texture=&T_Inky_Smoke_Tile2}.R))
  NamedRerouteDeclaration(None=StaticSwitchParameter(True=@n19, False=Multiply(A=@n19, B=@n21)) {parameterName=Juno}) {name=Alpha, nodeColor=<rgba 1 0.8672 0 1>}
  n22 = Multiply(A=Time(), B=ScalarParameter() {defaultValue=1, parameterName=UpDownSpeed})
  n23 = TextureSample(UVs=Multiply(A=TextureCoordinate() {coordinateIndex=1}, B=ScalarParameter() {defaultValue=1, parameterName=UpDownMaskTextureScale})) {samplerType=SAMPLERTYPE_LinearColor, texture=/Game/Effects/Fort_Effects/Textures/Tile/T_PerlinRGB_Displacement}
  n24 = ConstantBiasScale(None=Sine(None=Add(A=@n22, B=@n23.R)))
  n25 = Panner(
   Coordinate=TextureCoordinate() {coordinateIndex=1, uTiling=0.5, vTiling=0.5}
   Speed=Multiply(A=Constant2Vector() {g=-0.167, r=0.05}, B=ScalarParameter() {defaultValue=1, parameterName=UpDownSpeed})
  ) {speedX=0.05, speedY=-0.167}
  n26 = Saturate(None=Multiply(A=TextureSample(UVs=@n25) {texture=&T_Inky_Smoke_Tile2}.R) {constB=2})
  n27 = Multiply(
   A=LinearInterpolate(
    A=Constant3Vector() {constant=<rgba 0 0 -1 0>}
    B=Constant3Vector() {constant=<rgba 0 0 1 0>}
    Alpha=StaticSwitchParameter(True=@n24, False=@n26) {parameterName=Juno}
   )
   B=ScalarParameter() {defaultValue=2, parameterName=Z Offset Scale}
  )
  n28 = Multiply(
   A=TextureCoordinate() {coordinateIndex=1}
   B=ComponentMask(None=VectorParameter() {parameterName=FlickeringMaskTextureUVScale, defaultValue=<rgba 1 1 0 1>}.RGB) {g=true, r=true}
  )
  n29 = ComponentMask(None=VectorParameter() {parameterName=FlickeringMaskTextureUVOffset, defaultValue=<rgba 0 0 0 1>}.RGB) {g=true, r=true}
  n30 = Add(
   A=Multiply(A=Time(), B=ScalarParameter() {defaultValue=1, parameterName=FlikeringSpeed})
   B=TextureSample(UVs=Add(A=@n28, B=@n29)) {samplerType=SAMPLERTYPE_LinearColor, texture=&T_PerlinRGB_Displacement}.R
  )
  n31 = Power(Base=ConstantBiasScale(None=Sine(None=@n30)), Exp=ScalarParameter() {defaultValue=1, parameterName=FlickeringTextureMask_Power}) {constExponent=1}
  n32 = ScalarParameter() {defaultValue=5, parameterName=FlickeringTextureMask_Multiply}
  n33 = LinearInterpolate(B=ScalarParameter() {defaultValue=20, parameterName=Scale}, Alpha=Saturate(None=Multiply(A=@n31, B=@n32))) {constA=-15}
  n34 = WorldPosition()
  n35 = Constant() {r=10}
  n36 = LinearInterpolate(Alpha=NamedRerouteDeclaration(None=Power(Base=@n3.R) {constExponent=0.25}) {name=WPO_Alpha, nodeColor=<rgba 0.1953 0 1 1>}) {constA=5, constB=20}
  n37 = WorldPosition()
  n38 = StaticSwitchParameter(
   True=MaterialFunctionCall!(Center Location=WorldPosition().XYZ, 0-1 UVs, XY scale=Clamp(None=@n33) {maxDefault=100}, OverridePivot, "Normals (Optional)") {materialFunction=&Sprite}.World position offset
   False=Add(
    A=MaterialFunctionCall(Center Location=@n34.XYZ, XY scale=@n35) {materialFunction=&Sprite}.World position offset
    B=MaterialFunctionCall(Offset Amount=@n36, World Position=@n37.XYZ) {materialFunction=/Engine/Functions/Engine_MaterialFunctions02/WorldPositionOffset/CameraOffset}.Camera Offset
   )
  ) {parameterName=Juno}
  n39 = Subtract(A=ObjectPositionWS(), B=WorldPosition().XYZ)
  n40 = Time()
  n41 = ScalarParameter() {defaultValue=1, parameterName=FlikeringSpeed}
  n42 = TextureCoordinate() {coordinateIndex=1}
  n43 = ComponentMask(None=VectorParameter() {parameterName=FlickeringMaskTextureUVScale, defaultValue=<rgba 1 1 0 1>}.RGB) {g=true, r=true}
  n44 = VectorParameter() {parameterName=FlickeringMaskTextureUVOffset, defaultValue=<rgba 0 0 0 1>}
  n45 = TextureSample(UVs=Add(A=Multiply(A=@n42, B=@n43), B=ComponentMask(None=@n44.RGB) {g=true, r=true})) {samplerType=SAMPLERTYPE_LinearColor, texture=&T_PerlinRGB_Displacement}
  n46 = Fmod(A=Add(A=Multiply(A=@n40, B=@n41), B=Saturate(None=@n45.R)), B=Constant() {r=1})
  n47 = Multiply(A=Multiply(A=Multiply(A=@n39) {constB=-1}, B=ScalarParameter() {defaultValue=0.33, parameterName=Radial Offset Scale}), B=@n46)
  n48 = VectorParameter() {group=SpriteAnimTransition, parameterName=SpriteAnimTransition_Direction, defaultValue=<rgba 0 0 -30 1>}
  n49 = Multiply(A=@n46, B=@n48.RGB)
  n50 = ScalarParameter() {defaultValue=1.1, group=SpriteAnimTransition, parameterName=SpriteAnimTransition_TransitionTimeMultiply}
  n51 = Saturate(None=Subtract(
    A=Multiply(A=@n46, B=@n50) {constB=1.1}
    B=ScalarParameter() {group=SpriteAnimTransition, parameterName=SpriteAnimTransition_KeepRadialMotion}
   ))
  n52 = LinearInterpolate(
   A=StaticSwitchParameter(True=LinearInterpolate(A=@n47, B=@n49, Alpha=@n51), False=@n47) {group=SpriteAnimTransition, parameterName=SpriteAnimTransition}
   B=@n39
   Alpha=ScalarParameter() {bUseCustomPrimitiveData=true, parameterName=WPO_HideMesh Primitive}
  )
  n53 = Subtract(A=ObjectPositionWS(), B=WorldPosition().XYZ)
  n54 = Multiply(A=@n53) {constB=-1}
  n55 = ScalarParameter() {defaultValue=0.33, parameterName=Radial Offset Scale}
  n56 = OneMinus(None=ComponentMask(None=TextureCoordinate() {coordinateIndex=1}) {g=true})
  n57 = Multiply(A=Add(A=Time(), B=Multiply(A=@n56) {constB=5}), B=ScalarParameter() {defaultValue=1, parameterName=Radial Speed})
  n58 = Constant() {r=1}
  n59 = LinearInterpolate(
   A=Multiply(A=Multiply(A=@n54, B=@n55), B=Fmod(A=@n57, B=@n58))
   B=@n53
   Alpha=ScalarParameter() {bUseCustomPrimitiveData=true, parameterName=WPO_HideMesh Primitive}
  )
  NamedRerouteDeclaration(None=Add(A=Add(A=@n27, B=@n38) {constB=10}, B=StaticSwitchParameter(True=@n52, False=@n59) {parameterName=Juno})) {name=WPO, nodeColor=<rgba 1 0.5156 0 1>}
  n60 = TextureSample(UVs=Panner(Coordinate=TextureCoordinate() {coordinateIndex=1, uTiling=3, vTiling=0.1}) {speedY=-0.05}) {texture=&T_Inky_Smoke_Tile2}
  NamedRerouteDeclaration(None=Saturate(None=Multiply(A=@n60.R) {constB=4})) {name=Derived Random, nodeColor=<rgba 0.875 1 0 1>}
  NamedRerouteUsage()
 ]
}
