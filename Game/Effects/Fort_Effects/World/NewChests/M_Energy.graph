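# Additive, unlit, two-sided Niagara sprite material for a flowing energy effect, filed under the new chest effects.
# Panning Perlin noise distorts the UVs of smoke-shape and rainbow textures, with dynamic parameter Param1 offsetting their shared time.
# The smoke mask hue-shifts a base color, which adds to a saturation-adjustable rainbow layer and is tinted by particle color for emissive.
# The refined mask, depth fade and particle alpha form the opacity.
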Material /Game/Effects/Fort_Effects/World/NewChests/M_Energy
graph {
 properties {
  bUsedWithNiagaraSprites=true
  blendMode=BLEND_Additive
  pixelDepthOffsetMode=PDOM_Legacy
  shadingModel=MSM_Unlit
  twoSided=true
  usedShadingModels=Unlit
 }
 nodes [
  n1 = VectorParameter() {group=MainTex, parameterName=MainTex_UVControl, channelNames {a=VPan, b=UPan, g=VScale, r=UScale}, defaultValue=<rgba 1 1 0 0>}
  n2 = Add(A=DynamicParameter().Param1, B=Time() {bOverride_Period=true, period=6666})
  n3 = TextureObjectParameter() {
   group=NoiseA
   parameterName=NoiseA_Texture
   samplerType=SAMPLERTYPE_LinearGrayscale
   sortPriority=0
   texture=&T_SimplePerlin_Greyscale
  }
  n4 = VectorParameter() {group=NoiseA, parameterName=NoiseA_UVControl, channelNames {a=VPan, b=UPan, g=VScale, r=UScale}, defaultValue=<rgba 1.5 1 0 0.1>}
  n5 = MaterialFunctionCall(Texture=@n3, UVControl_RGB=@n4, UVControl_A=@n4, "[Override]TimeInput"=@n2) {materialFunction=&FXCommon_TextureSample}
  n6 = VectorParameter() {
   group=NoiseA
   parameterName=NoiseA_NoiseControl
   sortPriority=60
   channelNames {a=VStrength, b=UStrength, g=VOffset, r=UOffset}
   defaultValue=<rgba -0.5 -0.5 0.2 0.3>
  }
  n7 = Add(
   A=MaterialFunctionCall!(
    NoiseValue=MaterialFunctionCall(R=@n5.R, G=@n5.R) {materialFunction=&AppendMany}.RG
    Control_RGB=@n6
    Control_A=@n6
   ) {materialFunction=/Game/Effects/Niagara/MasterMaterials/Functions/Common/FXCommon_NoiseControl}.NoiseValue
   B=TextureCoordinate()
  )
  n8 = MaterialFunctionCall(
   Texture=TextureObject() {samplerType=SAMPLERTYPE_LinearGrayscale, texture=/Game/Tools/WaterfallGenerator/Art/T_Waterfall_SmokeShape}
   UVControl_RGB=@n1
   UVControl_A=@n1
   "[Override]TimeInput"=@n2
   "[Override]UVInput"=@n7
  ) {materialFunction=&FXCommon_TextureSample}
  n9 = MaterialFunctionCall(
   Hue Shift Percentage=Multiply(A=@n8.R, B=ScalarParameter() {defaultValue=-0.05, parameterName=H})
   Texture=VectorParameter() {parameterName=BaseC, defaultValue=<rgba 0.5 0.5 0.5 1>}.RGB
  ) {materialFunction=&HueShift}
  n10 = TextureObjectParameter() {group=RainbowColor, parameterName=RainbowColor_Texture, sortPriority=0, texture=&T_Tile_RainbowColor}
  n11 = VectorParameter() {group=RainbowColor, parameterName=RainbowColor_UVControl, channelNames {a=VPan, b=UPan, g=VScale, r=UScale}, defaultValue=<rgba 1 1 0 0>}
  n12 = Multiply(
   A=Desaturation(
    None=MaterialFunctionCall(Texture=@n10, UVControl_RGB=@n11, UVControl_A=@n11, "[Override]TimeInput"=@n2, "[Override]UVInput"=@n7) {materialFunction=&FXCommon_TextureSample}.RGB
    Fraction=ScalarParameter() {defaultValue=-0.5, parameterName=RainbowDes}
   )
   B=ScalarParameter() {defaultValue=1, parameterName=Rainbow}
  )
  n13 = ParticleColor()
  MaterialFunctionCall(Emissive=Multiply(A=Add(A=@n9.Result, B=@n12), B=@n13.RGB)) {materialFunction=&MF_FX_EmissiveAdjustments}
  n14 = VectorParameter() {parameterName=BaseFlowTexture_Refine, defaultValue=<rgba 10 1.3 0.2 0.5>}
  Multiply(
   A=@n13.A
   B=DepthFade(Opacity=MaterialFunctionCall(Texture=@n8.R, RGB=@n14.RGB, A=@n14.RGB) {materialFunction=&MF_FX_Refine}.Result)
  )
 ]
}
