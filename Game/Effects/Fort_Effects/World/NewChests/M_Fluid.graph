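# Translucent, unlit, two-sided fluid material for Niagara mesh particles in the new chest effects.
# A panning flow texture, sampled with V on both axes, is refined, edge-masked along V and depth-faded for opacity.
# Emissive adds a scaled, desaturation-adjusted panning rainbow texture to a base-to-root color blend driven by a power curve on V.
# SplineThicken, with base and tip width parameters, thickens spline-style mesh geometry.
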Material /Game/Effects/Fort_Effects/World/NewChests/M_Fluid
graph {
 properties {
  bUsedWithNiagaraMeshParticles=true
  blendMode=BLEND_Translucent
  pixelDepthOffsetMode=PDOM_Legacy
  shadingModel=MSM_Unlit
  twoSided=true
  usedShadingModels=Unlit
 }
 nodes [
  n1 = TextureObjectParameter() {
   group=BaseFlowTexture
   parameterName=BaseFlowTexture_Texture
   sortPriority=0
   texture=/Game/Effects/Fort_Effects/Textures/Smoke/T_Smoke_Trail_Soft_Chromatic
  }
  n2 = VectorParameter() {
   group=BaseFlowTexture
   parameterName=BaseFlowTexture_UVControl
   channelNames {a=VPan, b=UPan, g=VScale, r=UScale}
   defaultValue=<rgba 1 1 0.3 0>
  }
  n3 = MaterialFunctionCall(Float2=TextureCoordinate()) {materialFunction=&BreakOutFloat2Components}
  n4 = MaterialFunctionCall(
   Texture=@n1
   UVControl_RGB=@n2
   UVControl_A=@n2
   "[Override]UVInput"=MaterialFunctionCall(X=@n3.G, Y=@n3.G) {materialFunction=&MakeFloat2}.Result
  ) {materialFunction=&FXCommon_TextureSample}
  n5 = VectorParameter() {parameterName=BaseFlowTexture_Refine, defaultValue=<rgba 5 2 0.5 0.5>}
  n6 = VectorParameter() {
   group=BaseFlowTexture
   parameterName=BaseFlowTexture_EdgeMask_InvertControl
   channelNames {a="", b=Power, g=Multiply, r=Offset}
   defaultValue=<rgba 0 10 1.5 1>
  }
  n7 = VectorParameter() {
   group=BaseFlowTexture
   parameterName=BaseFlowTexture_EdgeMask_Control
   channelNames {a="", b=Power, g=Multiply, r=Offset}
   defaultValue=<rgba 0 1 2 1>
  }
  DepthFade(
   Opacity=Multiply(
    A=MaterialFunctionCall(Texture=@n4.R, RGB=@n5.RGB, A=@n5.RGB) {materialFunction=&MF_FX_Refine}.Result
    B=MaterialFunctionCall!(UorV=@n3.G, Invert_EdgeControl=@n6, NoInvert_EdgeControl=@n7) {materialFunction=/Game/Effects/Niagara/MasterMaterials/Functions/Common/FXCommon_EdgeMask_Ex}.Mask
   )
   FadeDistance=ScalarParameter() {defaultValue=3, parameterName=DepthFadeDistance}
  )
  MaterialFunctionCall(WidthBase=ScalarParameter() {defaultValue=30, parameterName=WidthBase}, WidthTip=ScalarParameter() {defaultValue=20, parameterName=WidthTip}) {materialFunction=&SplineThicken}
  n8 = TextureObjectParameter() {group=RainbowColor, parameterName=RainbowColor_Texture, sortPriority=0, texture=&T_Tile_RainbowColor}
  n9 = VectorParameter() {group=RainbowColor, parameterName=RainbowColor_UVControl, channelNames {a=VPan, b=UPan, g=VScale, r=UScale}, defaultValue=<rgba 1 1 0 0.1>}
  n10 = Desaturation(
   None=MaterialFunctionCall(Texture=@n8, UVControl_RGB=@n9, UVControl_A=@n9) {materialFunction=&FXCommon_TextureSample}.RGB
   Fraction=ScalarParameter() {defaultValue=-0.5, parameterName=RainbowDes}
  )
  n11 = ScalarParameter() {defaultValue=10, parameterName=A}
  n12 = MaterialFunctionCall(Float2=TextureCoordinate()) {materialFunction=&BreakOutFloat2Components}
  n13 = LinearInterpolate(
   A=VectorParameter() {parameterName=BaseColor, defaultValue=<rgba 2 2 2 1>}.RGB
   B=VectorParameter() {parameterName=RootColor, defaultValue=<rgba 2 2 2 1>}.RGB
   Alpha=Power(Base=@n12.G, Exp=ScalarParameter() {defaultValue=5, parameterName=RootEmissivePower})
  )
  MaterialFunctionCall(Emissive=Add(A=Multiply(A=@n10, B=@n11), B=@n13)) {materialFunction=&MF_FX_EmissiveAdjustments}
  Multiply(A=Normalize(VectorInput=@n4.RGB), B=@n13)
  Multiply(A=@n4.R)
  OneMinus(None=@n12.G)
  ScalarParameter()
 ]
}
