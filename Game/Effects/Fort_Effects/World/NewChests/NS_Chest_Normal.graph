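# Niagara effect for the normal world chest: a gold glow with drifting sparkles.
# Corona bursts three mesh layers, using spawn-group masks for look and VisibilityTag for renderer choice, with User FXScale and FXBrightnessScale adjusting some layers' size and alpha.
# GPU emitter Sprite_Gold continuously spawns two sprite groups with distance culling.
# It computes two spawn position and velocity sets, then masks by group between drifting motes and sparks pushed outward from a squashed sphere.
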NiagaraSystem /Game/Effects/Fort_Effects/World/NewChests/NS_Chest_Normal
dependencies {usedDataInterfaces [&NiagaraDataInterfaceCurve], usedDynamicInputs [], usedModules [&SystemState]}
emitters {
 Corona {
  emitterUpdateScript [
   &EmitterState()
   &SpawnBurst_Instantaneous(Use Spawn Probability=false, UseLoopCountLimit=true)
   &SpawnBurst_Instantaneous(Spawn Group=1, Use Spawn Probability=false, UseLoopCountLimit=true) {module=SpawnBurst_Instantaneous001}
   &SpawnBurst_Instantaneous(Spawn Group=2, Use Spawn Probability=false, UseLoopCountLimit=true) {module=SpawnBurst_Instantaneous002}
  ]
  particleSpawnScript [
   &InitializeParticle(
    Color=&MaskLinearColorBySpawnGroup(
     Masked Linear Color=<rgba 7 2.895 0 4>
     Passthrough Linear Color=&MaskLinearColorBySpawnGroup(Masked Linear Color=<rgba 1 1 1 1>, Passthrough Linear Color=<rgba 9 3.7221 0 2>, Spawn Group Mask=2)
     Spawn Group Mask=1
    )
    Mesh Scale=&MaskVectorBySpawnGroup(
     Masked Vector=&Multiply_VectorByFloat(Vector=<1.5 1 1.25>, Float=$User.FXScale)
     Passthrough Vector=&MaskVectorBySpawnGroup(
      Masked Vector=<1 1 1>
      Passthrough Vector=&VectorFromFloat(Value=&Multiply_Float(A=1.5, B=$User.FXScale))
      Spawn Group Mask=2
     )
     Spawn Group Mask=1
    )
    Mesh Scale Mode=::NewEnumerator4
    Position Offset=&MaskVectorBySpawnGroup(Masked Vector=<0 0 -9>, Passthrough Vector=<0 0 0>, Spawn Group Mask=1)
    UsePositionOffset=true
    Write Material Random=true
   )
   /SetVariables(Particles.VisibilityTag=$Emitter.SpawnGroup) {isSetParameters=true}
  ]
  particleUpdateScript [
   &ParticleState(Kill Particles When Lifetime Has Elapsed=false, Let Infinitely Lived Particles Die When Emitter Deactivates=true)
   &DynamicMaterialParameters(
    Index 0 Param 1=&MaskFloatBySpawnGroup(
     Masked Float=-2
     Passthrough Float=&MaskFloatBySpawnGroup(Masked Float=1, Passthrough Float=1.1, Spawn Group Mask=2)
     Spawn Group Mask=1
    )
    Index 0 Param 2=&MaskFloatBySpawnGroup(
     Masked Float=1
     Passthrough Float=&MaskFloatBySpawnGroup(Masked Float=1, Passthrough Float=0.25, Spawn Group Mask=2)
     Spawn Group Mask=1
    )
    Index 0 Param 3=1
    Index 0 Param 4=1
    Module.Param0WriteEnabled=true
    Module.Param1WriteEnabled=true
    Module.Param2WriteEnabled=true
    Module.Param3WriteEnabled=true
   )
   &ScaleColor(
    Scale Alpha=&MaskFloatBySpawnGroup(
     Masked Float=1
     Passthrough Float=&Multiply_Float(A=&MaskFloatBySpawnGroup(Masked Float=2.5, Passthrough Float=3, Spawn Group Mask=1), B=$User.FXBrightnessScale)
     Spawn Group Mask=2
    )
    ScaleA=true
    ScaleRGB=true
   )
  ]
  rendererClasses [&NiagaraMeshRendererProperties]
  renderers [
   {rendererClass=&NiagaraMeshRendererProperties, rendererIndex=0}
   {rendererClass=&NiagaraMeshRendererProperties, rendererIndex=1}
   {rendererClass=&NiagaraMeshRendererProperties, rendererIndex=2}
  ]
 }
 Sprite_Gold {
  simTarget=GPUComputeSim
  emitterUpdateScript [
   &EmitterState(
    Enable Distance Culling=true
    Enable Visibility Culling=true
    Inactive Response=::NewEnumerator2
    Life Cycle Mode=::NewEnumerator1
    Loop Duration=2
    MaxDistance=2500
    Module.bApplySpawnCountScaleByDistance=true
    Module.bUseMaxDistance=true
    Module.bUseMinDistance=false
    Scalability Mode=::NewEnumerator1
    Scale Spawn Count=true
    UseLoopDelay=false
   )
   &SpawnRate(SpawnRate=10, Use Spawn Probability=false)
   &SpawnRate(SpawnGroup=1, SpawnRate=&RandomRangeFloat(Minimum=5, Maximum=10), Use Spawn Probability=false) {module=SpawnRate001}
  ]
  particleSpawnScript [
   &InitializeParticle(
    Color=&MaskLinearColorBySpawnGroup(Masked Linear Color=<rgba 16 6.5 0 1>, Passthrough Linear Color=<rgba 16 12 9 1>)
    Lifetime Max=3
    Lifetime Min=0.6
    Lifetime Mode=::NewEnumerator1
    Position Offset=&MaskVectorBySpawnGroup(Masked Vector=&RandomRangeVector(Minimum=<35 25 40>, Maximum=<-35 -25 40>), Passthrough Vector=<0 0 0>)
    Sprite Size Mode=::NewEnumerator1
    Uniform Sprite Size Max=5
    Uniform Sprite Size Min=2
    UsePositionOffset=true
    Write Material Random=true
   )
   /SetVariables(Particles.OPosition:NiagaraPosition=$Particles.Position) {isSetParameters=true}
   &ShapeLocation(MaskSpawnGroup=false, Override Seed=false, Sphere Radius=22, UseOwnerScale=true)
   &AddVelocity(Velocity=&RandomRangeVector(Minimum=<50 50 0>, Maximum=<-50 0 0>))
   /SetVariables(Particles.APosition:NiagaraPosition=$Particles.Position) {isSetParameters=true}
   /SetVariables(Particles.AVelocity:Vector3f=$Particles.Velocity) {isSetParameters=true}
   /SetVariables(Particles.Position=$Particles.OPosition) {isSetParameters=true}
   /SetVariables() {isSetParameters=true}
   &SphereLocation(
    MaskSpawnGroup=false
    Non Uniform Scale=<1.6 1.333 1>
    Offset=<0 -10 44>
    Sphere Distribution:ENiagaraSphereDistributionMode=::NewEnumerator1
    Sphere Radius=22
    U Position:NiagaraFloat=&RandomRangeFloat(Minimum=0, Maximum=0.5)
    UseNonUniformScale=true
    UseOffset=true
    UseSphereOrientation=true
   )
   &AddVelocityFromPoint(
    Module.Use Falloff Distance=false
    Origin Offset=<0 0 35>
    Velocity Strength=&RandomRangeFloat(Minimum=5, Maximum=30)
   )
   /SetVariables(Particles.Position=&MaskVectorBySpawnGroup(
     Masked Vector=Linked() {linkedVariable {name=Particles.APosition, type=NiagaraPosition}}
     Passthrough Vector=Linked() {linkedVariable {name=Particles.Position, type=NiagaraPosition}}
    )) {isSetParameters=true}
   /SetVariables(Particles.Velocity=&MaskVectorBySpawnGroup(Masked Vector=$Particles.AVelocity, Passthrough Vector=$Particles.Velocity)) {isSetParameters=true}
   /SetVariables(Particles.VisibilityTag=$Emitter.SpawnGroup) {isSetParameters=true}
  ]
  particleUpdateScript [
   &ParticleState()
   &AccelerationForce(Acceleration=<0 0 -10>)
   &Drag(Use Linear Drag=true, Use Rotational Drag=true)
   &GravityForce(Gravity=&MakeVector(
     X=0
     Y=0
     Z=&MaskFloatBySpawnGroup(Masked Float=0, Passthrough Float=&RandomRangeFloat(Minimum=-10, Maximum=-25))
    ))
   &SolveForcesAndVelocity(Clamp Velocity=false, Limit Acceleration=false)
   &ScaleSpriteSize(Uniform Curve Sprite Scale=DataInterface() {propertyValues {Curve {keys [
        {arriveTangent=9.0725, interpMode=RCIM_Cubic, leaveTangent=3.7536, tangentMode=RCTM_Break, time=0, value=0}
        {arriveTangent=1.0728, interpMode=RCIM_Cubic, tangentMode=RCTM_Break, time=0.2, value=0.5}
        {arriveTangent=-0.625, interpMode=RCIM_Cubic, leaveTangent=-0.625, time=0.8, value=0.9847}
        {arriveTangent=-0.0405, interpMode=RCIM_Cubic, leaveTangent=-0.0405, tangentMode=RCTM_User, time=1, value=0}
       ]}}})
  ]
  rendererClasses [&NiagaraSpriteRendererProperties]
  renderers [
   {rendererClass=&NiagaraSpriteRendererProperties}
   {rendererClass=&NiagaraSpriteRendererProperties, rendererIndex=1}
  ]
 }
}
summary {
 systemName=NS_Chest_Normal
 emitters [
  {emitterName=Corona, rendererClasses [&NiagaraMeshRendererProperties]}
  {emitterName=Sprite_Gold, simTarget=GPUComputeSim, rendererClasses [&NiagaraSpriteRendererProperties]}
 ]
 userVariables [
  User.FXBrightnessScale {type=NiagaraFloat, defaultValue=1, description=""}
  User.FXScale {type=NiagaraFloat, defaultValue=1, description=""}
 ]
}
